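# Sound cue for the outgoing sound of a rotary phone, named as part of an Athena quest in Figment Season 10.
# A concatenator plays two outgoing waves back to back, with the second at half the input volume of the first.
# A root modulator slightly randomizes pitch and lowered volume for the whole sequence.
# It borrows the gadget-activate attenuation and routes to the in-game mission sound class.
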SoundCue /Figment_S10/Athena/Sounds/Quests/RotaryPhone/Athena_Quest_Rotary_Outgoing_Cue
root=Modulator_0
nodes [
 {id=Concatenator_0, children [WavePlayer_0,WavePlayer_3], properties {inputVolume [1,0.5]}}
 {id=Modulator_0, children [Concatenator_0], properties {pitchMax=1, pitchMin=0.97, volumeMax=0.5, volumeMin=0.4}}
 {id=WavePlayer_0, properties {soundWaveAssetPtr=&/Athena_Quest_Rotary_Outgoing_01}}
 {id=WavePlayer_3, properties {soundWaveAssetPtr=&/Athena_Quest_Rotary_Outgoing_02}}
]
properties {
 attenuationSettings=&Fort_GadgetActivate_Att
 maxDistance=4000
 soundClassObject=/Game/Sounds/FortSoundClassesAndModes/Fort_Mission_IngameSounds
 volumeMultiplier=0.5
}
