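# Sound cue for a quest rotary phone, playing a return-end sound that, judging by its name, marks the dial finishing its spin back.
# A Random node picks one of three equally weighted wave variations and feeds a Modulator at default settings.
# Playback uses the shared gadget-activate attenuation and the mission in-game sounds class.
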SoundCue /Figment_S10/Athena/Sounds/Quests/RotaryPhone/Athena_Quest_Rotary_ReturnEnd_Cue
root=Modulator_0
nodes [
 {id=Modulator_0, children [Random_1]}
 {id=Random_1, children [WavePlayer_0,WavePlayer_3,WavePlayer_4], properties {weights [1,1,1]}}
 {id=WavePlayer_0, properties {soundWaveAssetPtr=&/Athena_Quest_Rotary_ReturnEnd_01}}
 {id=WavePlayer_3, properties {soundWaveAssetPtr=&/Athena_Quest_Rotary_ReturnEnd_02}}
 {id=WavePlayer_4, properties {soundWaveAssetPtr=&/Athena_Quest_Rotary_ReturnEnd_03}}
]
properties {
 attenuationSettings=&Fort_GadgetActivate_Att
 maxDistance=4000
 soundClassObject=/Game/Sounds/FortSoundClassesAndModes/Fort_Mission_IngameSounds
 volumeMultiplier=1
}
