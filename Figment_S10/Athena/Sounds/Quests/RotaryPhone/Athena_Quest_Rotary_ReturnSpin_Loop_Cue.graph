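# Sound cue for the looping return-spin sound of a rotary phone in a Figment season 10 quest.
# A looping wave player feeds a modulator that varies pitch slightly downward from normal.
# It plays at reduced volume in the mission in-game sound class, using the shared gadget-activate attenuation.
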SoundCue /Figment_S10/Athena/Sounds/Quests/RotaryPhone/Athena_Quest_Rotary_ReturnSpin_Loop_Cue
root=Modulator_0
nodes [
 {id=Modulator_0, children [WavePlayer_1], properties {pitchMax=1, pitchMin=0.97}}
 {
  id=WavePlayer_1
  properties {bLooping=true, soundWaveAssetPtr=&/Athena_Quest_Rotary_ReturnSpin_Loop}
 }
]
properties {
 attenuationSettings=&Fort_GadgetActivate_Att
 maxDistance=4000
 soundClassObject=/Game/Sounds/FortSoundClassesAndModes/Fort_Mission_IngameSounds
 volumeMultiplier=0.25
}
